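# Skeletal-mesh material for the Maxis vehicle wheel, mixing a cel-shaded emissive look with PBR surfaces.
# Vertex color red is declared as the rim/tire selector for paired diffuse, normal and mask textures.
# Cel bands come from stepping the normal-to-camera dot product with Ceil, while a mask channel separates unlit and PBR color, metallic and roughness.
# The resulting material attributes pass through shared vehicle damage and de-rez transition functions.
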Material /VehicleCosmetics/Wheels/Maxis/Materials/M_Wheel_Maxis
graph {
 properties {
  bAutomaticallySetUsageInEditor=false
  bUseMaterialAttributes=true
  bUsedWithMeshDeformer=true
  bUsedWithSkeletalMesh=true
  previewMesh=/VehicleCosmetics/Wheels/Maxis/Meshes/SK_Wheel_Maxis
  usedShadingModels=Default Lit
 }
 nodes [
  n1 = NamedRerouteUsage()
  n2 = SetMaterialAttributes!(
   MaterialAttributes
   Base Color=NamedRerouteUsage()
   Emissive Color=Subtract(A=NamedRerouteUsage(), B=NamedRerouteUsage())
   Metallic=LinearInterpolate(A=ScalarParameter() {parameterName=Unlit Metallic}, B=ScalarParameter() {defaultValue=0.5, parameterName=PBR Metallic}, Alpha=@n1)
   Roughness=LinearInterpolate(
    A=ScalarParameter() {defaultValue=1, parameterName=Unlit Roughness}
    B=ScalarParameter() {defaultValue=0.2, parameterName=PBR Roughness}
    Alpha=@n1
   )
  ) {attributeSetTypes [
    #1
    #2
    #3
    #4
   ]}
  GetMaterialAttributes(None=MaterialFunctionCall(In=MaterialFunctionCall(Material=@n2) {materialFunction=&MF_BaseVehicle_Damage}.Result) {materialFunction=&MF_VehicleDeRez_Transition}.Result) {attributeGetTypes [#5]}
  n3 = Constant3Vector() {constant=<rgba 0 0 1 0>}
  n4 = TextureSampleParameter2D() {
   group=00_Base
   parameterName=Normal Rim
   samplerSource=SSM_Wrap_WorldGroupSettings
   samplerType=SAMPLERTYPE_Normal
   sortPriority=0
   texture=/VehicleCosmetics/Wheels/Maxis/Textures/T_Maxis_Rim_N
  }
  n5 = TextureSampleParameter2D() {
   group=00_Base
   parameterName=Normal Tire
   samplerSource=SSM_Wrap_WorldGroupSettings
   samplerType=SAMPLERTYPE_Normal
   sortPriority=0
   texture=/VehicleCosmetics/Wheels/Maxis/Textures/T_Maxis_Tire_N
  }
  n6 = LinearInterpolate(A=@n4.RGB, B=@n5.RGB, Alpha=NamedRerouteDeclaration(None=VertexColor().R) {name=VC_R, nodeColor=<rgba 1 0 0 1>})
  n7 = ScalarParameter() {defaultValue=1, group=00_Base, parameterName=Cel Shading Use Normal}
  n8 = DotProduct(A=Transform(None=LinearInterpolate(A=@n3, B=@n6, Alpha=@n7)), B=CameraVectorWS())
  n9 = ScalarParameter() {defaultValue=2, group=00_Base, parameterName=Cel Shading Step Count, sortPriority=2}
  n10 = Ceil(None=Subtract(A=Multiply(A=@n8, B=@n9), B=ScalarParameter() {defaultValue=0.2, group=00_Base, parameterName=Cel Shading Step Amount, sortPriority=2}))
  NamedRerouteDeclaration(None=Clamp(None=ConstantBiasScale(None=@n10) {bias=-0.4, scale=1})) {name=CeilShading, nodeColor=<rgba 0.8516 1 0 1>}
  NamedRerouteDeclaration(None=LinearInterpolate(
    A=VectorParameter() {parameterName=Unlit Color, defaultValue=<rgba 0 0 0 1>}.RGB
    B=VectorParameter() {parameterName=PBR Color, defaultValue=<rgba 1 0.1946 0 1>}.RGB
    Alpha=NamedRerouteUsage()
   )) {name=BC, nodeColor=<rgba 1 0.6094 0 1>}
  n11 = TextureSampleParameter2D() {
   group=00_Base
   parameterName=Diffuse Rim
   samplerSource=SSM_Wrap_WorldGroupSettings
   sortPriority=0
   texture=/VehicleCosmetics/Wheels/Maxis/Textures/T_Maxis_Rim_BC
  }
  n12 = TextureSampleParameter2D() {
   group=00_Base
   parameterName=Diffuse Tire
   samplerSource=SSM_Wrap_WorldGroupSettings
   sortPriority=0
   texture=/VehicleCosmetics/Wheels/Maxis/Textures/T_Maxis_Tire_BC
  }
  n13 = NamedRerouteUsage()
  NamedRerouteDeclaration(None=Multiply(A=LinearInterpolate(A=@n11.RGB, B=@n12.RGB, Alpha=@n13), B=NamedRerouteUsage())) {name=Emissive, nodeColor=<rgba 0 0.1328 1 1>}
  n14 = TextureSampleParameter2D() {
   group=00_Base
   parameterName=Masks Rim
   samplerSource=SSM_Wrap_WorldGroupSettings
   samplerType=SAMPLERTYPE_LinearColor
   sortPriority=0
   texture=/VehicleCosmetics/Wheels/Maxis/Textures/T_Maxis_Rim_M
  }
  n15 = TextureSampleParameter2D() {
   group=00_Base
   parameterName=Masks Tire
   samplerSource=SSM_Wrap_WorldGroupSettings
   samplerType=SAMPLERTYPE_LinearColor
   sortPriority=0
   texture=/VehicleCosmetics/Wheels/Maxis/Textures/T_Maxis_Tire_M
  }
  n16 = NamedRerouteUsage()
  NamedRerouteDeclaration(None=ComponentMask(None=LinearInterpolate(A=@n14.RGB, B=@n15.RGB, Alpha=@n16)) {r=true}) {name=PBR Mask, nodeColor=<rgba 0 1 0.9297 1>}
 ]
}
